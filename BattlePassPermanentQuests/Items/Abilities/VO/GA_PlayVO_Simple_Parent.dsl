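; Parent gameplay ability that plays a battle pass quest VO sound sequence on clients able to run cosmetic events.
; It waits until the VO-playing tag is absent, then adds it as an unreplicated loose tag to prevent overlap.
; Each attached sound's finish event advances the index, after any per-entry delay, until no valid sound remains.
; Ending clears the timer, sends a StoppedListeningToVO event and removes the tag.
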(event Ability|EventActivateAbility
  ;[0 0 0] Activate - assume we want to play the audio for this client
  (|CallParentFunction)
  ;;[1 1 1] Since the VO will only play on the client, should be safe to use a client-only tag for checking if it is playing or not so we don't have to spend resources replicating it
  (Networking|CanExecuteCosmeticEvents
    (:True
      (Ability|Tasks|WaitGameplayTagRemove (Variables|Default|GetVOPlayingTag) 0 true
        (:then
          (Development|PrintString "Checking if player is listening to VO... waiting until it has finished" true true "(R=1.000000,G=0.000000,B=0.484915,A=1.000000)" 5.0))
        (:Removed
          (Development|PrintString "Player is not listening to VO. Starting.." true true "(R=1.000000,G=0.000000,B=0.484915,A=1.000000)" 5.0)
          (Ability|GameplayEffect|AddLooseGameplayTags (CallFunction|GetAvatarasFortPlayerPawn) (GameplayTags|MakeGameplayTagContainerfromTag (Variables|Default|GetVOPlayingTag)))
          (CallFunction|PlaySoundSequence (Variables|SoundSequence|GetCurrentIndex)))))))

(event Custom|StartDelay
  ;[0 0 0] Start a delay between the current audio file playing and the next if it is set in the sequence asset
  (bind _sequencedata (CallFunction|GetSoundSequenceatIndex (Variables|SoundSequence|GetCurrentIndex)))
  (bind (_sequencedata_sound _sequencedata_delay) (Utilities|Struct|BreakSoundSequenceData _sequencedata))
  (if (== _sequencedata_delay 0.0)
    (CallFunction|AdvanceSequence)
    (else
      (bind _returnvalue (Utilities|Time|SetTimerbyEvent (EventDispatchers|CreateEvent :!event Custom|AdvanceSequence) _sequencedata_delay false true))
      (Variables|Default|SetDelayTimerHandle _returnvalue))))

(event Custom|AdvanceSequence
  (bind _soundspawned (Variables|Default|GetSoundSpawned))
  ;[0 0 0] Helper event for advancing the sequence to the next audio file
  (Utilities|IsValid _soundspawned
    (:Is%20Valid
      (EventDispatchers|UnbindEventfromOnAudioFinished _soundspawned (EventDispatchers|CreateEvent :!event Custom|StartDelay)))
    (:Is%20Not%20Valid
      _))
  (bind _result (Math|Integer|IncrementInt (Variables|SoundSequence|GetCurrentIndex))
    (:%20%20
      (CallFunction|PlaySoundSequence _result))))

(event Ability|EventOnEndAbility
  (params (bWasCancelled :type Boolean))
  ;[0 0 0] End play - clear out any timers if they were actively running
  (|CallParentFunction bWasCancelled)
  (CallFunction|ClearTimerHandleifValid (Variables|Default|GetDelayTimerHandle))
  (Ability|SendGameplayEventtoActor (CallFunction|GetAvatarasFortPlayerPawn) "(TagName=\"Event.Quest.StoppedListeningToVO\")" (Utilities|Struct|MakeGameplayEventData "(TagName=\"\")" 0 0 (Variables|SoundSequence|GetSoundSequenceToPlay)))
  ;;[1 1 1] Remove client-only tag for the player listening to VO
  (Networking|CanExecuteCosmeticEvents
    (:True
      (Ability|GameplayEffect|RemoveLooseGameplayTags (CallFunction|GetAvatarasFortPlayerPawn) (GameplayTags|MakeGameplayTagContainerfromTag (Variables|Default|GetVOPlayingTag))))))

(event Custom|ClearTimerHandleIfValid
  (params (TimerHandle :type Timer%20Handle%20Structure))
  ;[0 0 0] End play - clear out any timers if they were actively running
  (if (Utilities|Time|IsValidTimerHandle TimerHandle)
    (Utilities|Time|ClearandInvalidateTimerbyHandle TimerHandle)))

(fn GetSoundSequenceAtIndex
  (params (Index :type Integer))
  (bind _soundsequencetoplay (Variables|SoundSequence|GetSoundSequenceToPlay))
  ;[0 0 0] Get the audio file in the sound sequence for the index provided
  (Utilities|IsValid _soundsequencetoplay
    (:Is%20Valid
      (bind (_outsoundsequencedata _returnvalue) (Class|FortSoundSequence|GetSoundSequenceDataforIndex _soundsequencetoplay Index))
      (if _returnvalue
        (return _outsoundsequencedata)
        (else
          _)))
    (:Is%20Not%20Valid
      _))
  (Development|PrintString "Sound sequence for S39 play audio not valid!" true true "(R=1.000000,G=0.799738,B=0.000000,A=1.000000)" 4.0)
  (return 0))

(fn PlaySoundSequence
  (params (Index :type Integer))
  ;[0 0 0] Play the audio file associated with the given index, if valid
  (bind _sequencedata (CallFunction|GetSoundSequenceatIndex Index))
  (bind _sequencedata_sound (Utilities|Struct|BreakSoundSequenceData _sequencedata))
  (Utilities|IsValid _sequencedata_sound
    (:Is%20Valid
      (bind _returnvalue (Audio|SpawnSoundAttached _sequencedata_sound (Variables|/Script/Engine.Actor|GetRootComponent (Ability|GetAvatarActorfromActorInfo))))
      (bind _output_get (Variables|Default|SetSoundSpawned _returnvalue))
      (EventDispatchers|BindEventtoOnAudioFinished _output_get (EventDispatchers|CreateEvent :!event StartDelay)))
    (:Is%20Not%20Valid
      (Ability|EndAbility))))
